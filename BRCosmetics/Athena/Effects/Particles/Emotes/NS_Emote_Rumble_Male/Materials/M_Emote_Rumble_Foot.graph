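# Unlit, translucent, two-sided Niagara mesh-particle material for the male Rumble emote's foot effect, colored by particle color through emissive adjustments.
# Polar UVs are reshaped, scaled and panned with a per-particle time offset to scroll a smoke texture, sampled with explicit derivatives (DDX feeds both inputs).
# That texture feeds an erosion function driven by the radial coordinate plus an Erosion dynamic parameter.
# Opacity multiplies the result by particle alpha and a facing-ratio fade.
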Material /BRCosmetics/Athena/Effects/Particles/Emotes/NS_Emote_Rumble_Male/Materials/M_Emote_Rumble_Foot
graph {
 properties {
  bUsedWithNiagaraMeshParticles=true
  blendMode=BLEND_Translucent
  pixelDepthOffsetMode=PDOM_Legacy
  previewMesh=/BRCosmetics/Athena/Effects/Particles/Emotes/NS_Emote_Rumble_Male/Meshes/SM_Emote_Rumble_Foot_00
  shadingModel=MSM_Unlit
  twoSided=true
  usedShadingModels=Unlit
 }
 nodes [
  n1 = ParticleColor()
  MaterialFunctionCall(Emissive=@n1.RGB) {materialFunction=&MF_FX_EmissiveAdjustments}
  n2 = MaterialFunctionCall!(UV=TextureCoordinate(), Twist Center, Twist Edges) {materialFunction=/Game/Accessories/FORT_Backpacks/Backpack_F_MED_Celestial_Glow/FX/Materials/MF_PolarUV}
  n3 = MaterialFunctionCall(Float2=@n2.Polar UVs) {materialFunction=&BreakOutFloat2Components}
  n4 = Power(Base=@n3.G, Exp=ScalarParameter() {defaultValue=0.666, parameterName=V Exponent})
  n5 = ScalarParameter() {defaultValue=2, parameterName=U Scale}
  n6 = ScalarParameter() {defaultValue=0.3, parameterName=V Scale}
  n7 = Panner(
   Coordinate=Multiply(A=AppendVector(A=@n3.R, B=@n4), B=AppendVector(A=@n5, B=@n6))
   Time=Add(A=ParticleRandom(), B=Time())
   Speed=AppendVector(A=Constant(), B=ScalarParameter() {defaultValue=-1.618, parameterName=V Speed})
  )
  n8 = MaterialFunctionCall!(
   Gradient=Power(
    Base=TextureSample!(UVs=@n7, Tex, "DDX(UVs)"=@n2.DDX, "DDY(UVs)"=@n2.DDX, Apply View MipBias) {mipValueMode=TMVM_Derivative, texture=/Game/Global/Textures/Default/Tileable_Patterns/T_Blowing_Smoke_Pattern2_Tile_DISP}.R
    Exp=ScalarParameter() {defaultValue=0.45, parameterName=Texture Exponent}
   )
   Erosion Amount=Add(
    A=ComponentMask(None=@n2.Polar UVs) {g=true}
    B=DynamicParameter() {defaultValue=<rgba 0 0 0 0>, paramNames [Erosion, - unused -, - unused -, - unused -]}.Erosion
   )
   Softness=ScalarParameter() {defaultValue=0.666, parameterName=Softness}
  ) {materialFunction=/BRCosmetics/Athena/Effects/Particles/Skydiving/NS_Contrail_DarkAndLight/Functions/MF_Erosion}
  n9 = ScalarParameter() {defaultValue=3, parameterName=Final Exponent}
  n10 = Abs(None=DotProduct(A=VertexNormalWS(), B=CameraVectorWS()))
  n11 = ScalarParameter() {defaultValue=2, parameterName=Fresnel Exponent}
  Multiply(A=@n1.A, B=Multiply(A=Power(Base=@n8.Increased Range, Exp=@n9), B=Power(Base=@n10, Exp=@n11)))
 ]
}
