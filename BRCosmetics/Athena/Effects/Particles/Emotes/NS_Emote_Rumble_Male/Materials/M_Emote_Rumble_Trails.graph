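# Translucent unlit trail material for Niagara mesh particles in the male Rumble emote, colored by particle color.
# Dynamic parameters, named for left and right splines, drive reveal mask, end fades, thickness and UV scaling.
# Opacity multiplies airflow and soft smoke textures by those masks and particle alpha, with quality-switched depth fade.
# A cross product of camera direction and transformed Y axis builds a camera-facing width offset, scaled by thickness.
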Material /BRCosmetics/Athena/Effects/Particles/Emotes/NS_Emote_Rumble_Male/Materials/M_Emote_Rumble_Trails
graph {
 properties {
  bUsedWithNiagaraMeshParticles=true
  blendMode=BLEND_Translucent
  pixelDepthOffsetMode=PDOM_Legacy
  shadingModel=MSM_Unlit
  usedShadingModels=Unlit
 }
 nodes [
  n1 = ParticleColor()
  MaterialFunctionCall(Emissive=@n1.RGB) {materialFunction=&MF_FX_EmissiveAdjustments}
  n2 = TextureCoordinate()
  n3 = ComponentMask(None=@n2) {r=true}
  n4 = ConstantBiasScale(None=ComponentMask(None=TextureCoordinate() {coordinateIndex=1}) {r=true}) {bias=-0.5, scale=2}
  n5 = Add(A=Frac(None=ComponentMask(None=@n2) {g=true}))
  n6 = DynamicParameter() {paramNames [Left Spline Mask Reveal, Right Spline Mask Reveal, Left Spline Mask Size, Right Spline Mask Size]}
  n7 = ComponentMask(None=TextureCoordinate() {coordinateIndex=1}) {r=true}
  n8 = LinearInterpolate(A=@n6.Left Spline Mask Reveal, B=@n6.Left Spline Mask Reveal, Alpha=@n7)
  n9 = Divide(A=OneMinus(None=Subtract(A=@n5, B=@n8)), B=LinearInterpolate(A=@n6.Left Spline Mask Size, B=@n6.Left Spline Mask Size, Alpha=@n7))
  n10 = Saturate(None=@n9)
  n11 = DynamicParameter() {
   parameterIndex=3
   defaultValue=<rgba 1 0.4 0.8 1>
   paramNames [Airflow Texture Selector, Airflow UV Exponent, Airflow UV Scale, Opacity Exponent]
  }
  n12 = Multiply(A=Subtract(A=Power(Base=@n10, Exp=@n11.Airflow UV Exponent)) {constB=0.5}, B=@n11.Airflow UV Scale)
  n13 = TextureSample(UVs=AppendVector(A=Multiply(A=@n3, B=@n4), B=Add(A=@n12) {constB=0.5})) {texture=/Game/Effects/Fort_Effects/Emotes/Skirmish/Skirmish_Male/T_Skirmish_airflow}
  n14 = DynamicParameter() {
   parameterIndex=3
   defaultValue=<rgba 1 0.4 0.8 1>
   paramNames [Airflow Texture Selector, Airflow UV Exponent, Airflow UV Scale, Opacity Exponent]
  }
  n15 = Saturate(None=@n14.Airflow Texture Selector)
  n16 = Saturate(None=OneMinus(None=@n9))
  n17 = TextureCoordinate()
  n18 = ComponentMask(None=TextureCoordinate() {coordinateIndex=1}) {r=true}
  n19 = AppendVector(A=Constant(), B=Divide(A=@n18))
  n20 = Constant() {r=1}
  n21 = DynamicParameter() {parameterIndex=1, paramNames [Left Spline Thickness, Right Spline Thickness, Left Spline UV Scale, Right Spline UV Scale]}
  n22 = LinearInterpolate(A=@n21.Left Spline UV Scale, B=@n21.Left Spline UV Scale, Alpha=@n18)
  n23 = MaterialFunctionCall(Float2=Multiply(A=Add(A=@n17, B=@n19), B=AppendVector(A=@n20, B=@n22))) {materialFunction=&BreakOutFloat2Components}
  n24 = Multiply(
   A=Multiply(A=LinearInterpolate(A=@n13.R, B=@n13.R, Alpha=@n15), B=Frac(None=@n16) {desc=Mask})
   B=TextureSample(UVs=AppendVector(A=@n23.G, B=@n23.G)) {samplerType=SAMPLERTYPE_Grayscale, texture=&T_Smoke_Trail_Soft}.R
  )
  n25 = ComponentMask(None=TextureCoordinate()) {g=true}
  n26 = DynamicParameter() {
   parameterIndex=2
   defaultValue=<rgba 0 0 0 0>
   paramNames [Left Spine Bottom Fade Size, Right Spline Bottom Fade Size, Left Spline Top Fade Size, Right Spline Top Fade Size]
  }
  n27 = ComponentMask(None=TextureCoordinate() {coordinateIndex=1}) {r=true}
  n28 = LinearInterpolate(A=@n26.Left Spine Bottom Fade Size, B=@n26.Left Spine Bottom Fade Size, Alpha=@n27)
  n29 = OneMinus(None=@n25)
  n30 = LinearInterpolate(A=@n26.Left Spline Top Fade Size, B=@n26.Left Spline Top Fade Size, Alpha=@n27)
  n31 = Multiply(A=Saturate(None=Divide(A=@n25, B=@n28)), B=Saturate(None=Divide(A=@n29, B=@n30)))
  n32 = DynamicParameter() {
   parameterIndex=3
   defaultValue=<rgba 1 0.4 0.8 1>
   paramNames [Airflow Texture Selector, Airflow UV Exponent, Airflow UV Scale, Opacity Exponent]
  }
  n33 = Power(Base=Multiply(A=Multiply(A=@n24, B=@n31), B=@n1.A), Exp=@n32.Opacity Exponent)
  n34 = DepthFade(Opacity=@n33, FadeDistance=ScalarParameter() {defaultValue=33, parameterName=Depth Fade Distance})
  QualitySwitch(Default=@n34, Low=@n33, High=@n34, Medium=@n33, Epic=@n34)
  n35 = CameraPositionWS()
  n36 = WorldPosition()
  n37 = Constant3Vector() {constant=<rgba 0 1 0 0>}
  n38 = CrossProduct(A=Normalize(VectorInput=Subtract(A=@n35, B=@n36.XYZ)), B=Normalize(VectorInput=Transform(None=@n37)))
  n39 = Subtract(A=ComponentMask(None=TextureCoordinate()) {r=true}) {constB=0.5}
  n40 = DynamicParameter() {parameterIndex=1, paramNames [Left Spline Thickness, Right Spline Thickness, Left Spline UV Scale, Right Spline UV Scale]}
  n41 = LinearInterpolate(A=@n40.Left Spline Thickness, B=@n40.Left Spline Thickness, Alpha=ComponentMask(None=TextureCoordinate() {coordinateIndex=1}) {r=true})
  Reroute(None=Multiply(A=Normalize(VectorInput=@n38), B=Multiply(A=@n39, B=@n41)))
 ]
}
